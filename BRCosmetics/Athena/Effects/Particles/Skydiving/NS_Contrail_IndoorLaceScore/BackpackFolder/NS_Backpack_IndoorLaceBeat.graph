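# Niagara effect for the IndoorLaceBeat backpack cosmetic: a glowing clock with a galaxy, orbiting motes, sparkles, dust and waves.
# Clock, Galaxy and dust emitters follow the skeletal mesh; Clock layers sprites, mesh and light.
# System timelines, alongside a fake audio-reactivity loop module, modulate brightness, rotation and spawn rates.
# User.VariantPicker selects purple or gold palettes, with time-of-day scaling on brightness and opacity.
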NiagaraSystem /BRCosmetics/Athena/Effects/Particles/Skydiving/NS_Contrail_IndoorLaceScore/BackpackFolder/NS_Backpack_IndoorLaceBeat
dependencies {
 usedDataInterfaces [
  &NiagaraDataInterfaceColorCurve
  &NiagaraDataInterfaceCurve
  &NiagaraDataInterfacePlatformSet
  &NiagaraDataInterfaceSkeletalMesh
  &NiagaraDataInterfaceArrayFloat
 ]
 usedDynamicInputs [
  &MakeFloatFromVector2d
  &ColorFromCurve
  &ModuloFloat
  /Niagara/DynamicInputs/Lerp/SmoothLerpOverTime_Float
  &MakeCustomFloatFromBool
  &MakeFloatFromLinearColor
  &RemapRange
 ]
 usedModules [
  &self:SystemSpawnScript.SystemScriptSource.SystemScriptGraph.NiagaraNodeAssignment_5.SetVariables_#1
  &self:SystemSpawnScript.SystemScriptSource.SystemScriptGraph.NiagaraNodeAssignment_9.SetVariables_#2
  &SystemState
  &self:SystemSpawnScript.SystemScriptSource.SystemScriptGraph.NiagaraNodeAssignment_3.SetVariables_#3
  &self:SystemSpawnScript.SystemScriptSource.SystemScriptGraph.NiagaraNodeAssignment_1.SetVariables_#4
  &Timeline
  /Game/Effects/Niagara/Modules/Audio/FN_FakeAudioReactivityLoop
  &self:SystemSpawnScript.SystemScriptSource.SystemScriptGraph.NiagaraNodeAssignment_10.SetVariables_#5
  &self:SystemSpawnScript.SystemScriptSource.SystemScriptGraph.NiagaraNodeAssignment_2.SetVariables_#6
  &self:SystemSpawnScript.SystemScriptSource.SystemScriptGraph.NiagaraNodeAssignment_11.SetVariables_#7
 ]
}
emitters {
 Clock {
  emitterUpdateScript [
   &EmitterState(
    Inactive Response=::NewEnumerator1
    Life Cycle Mode=::NewEnumerator1
    Loop Behavior=::NewEnumerator1
    Loop Duration Mode=::NewEnumerator1
    UseLoopDelay=true
   )
   &SpawnBurst_Instantaneous(Spawn Count=1, Use Spawn Probability=false)
  ]
  particleSpawnScript [
   &InitializeParticle(
    Color=&MergeHelper_ColorSwitch(
     0=<rgba 0.586 0.0626 4 1>
     1=<rgba 2 0.5449 0.0313 1>
     2=?
     3=?
     Index=&MakeIntFromFloat(Float=$User.VariantPicker, Conversion Mode=::NewEnumerator1)
     Num=::NewEnumerator0
    )
    Lifetime=2
    Sprite Size Mode=::NewEnumerator3
    Uniform Sprite Size=30
    UsePositionOffset=true
    Write Material Random=true
   ) {module=InitializeParticle001}
   &InitialMeshOrientation(Mesh Axis To Orient=<0 1 0>, Mesh Orientation Mode=::NewEnumerator2, Orientation Vector=<1 0 -0.03>, Use Rotation Vector=false)
  ]
  particleUpdateScript [
   &ParticleState(Kill Particles When Lifetime Has Elapsed=false, Let Infinitely Lived Particles Die When Emitter Deactivates=true)
   &ScaleSpriteSize(
    Scale Sprite Size Mode=::NewEnumerator0
    Uniform Scale Factor=&MakeFloatFromVector(Vector=$Engine.Owner.Scale, Channel=::NewEnumerator0)
   )
   &SkeletalMeshLocation(MaskSpawnGroup=false, Mesh Sampling Type=::NewEnumerator5, Skeletal Mesh=$System.Skeletal Mesh Reader, Socket Sampling Mode=::NewEnumerator5)
   &DynamicMaterialParameters(
    Index 0 Param 1=2
    Index 0 Param 2=1
    Index 0 Param 3=1
    Index 0 Param 4=1
    Module.Index0Float0Write=true
    Module.Index0Float1Write=true
    Module.Index0Float2Write=true
    Module.Index0Float3Write=true
   )
   &ScaleColor(
    Scale Alpha=&FNTimeOfDayScalar(Night Scale=2, Morning Scale=3, Evening Scale=3)
    ScaleA=true
    ScaleRGB=true
   ) {module=ScaleColor002}
   /SetVariables(
    Particles.GlowColor=&MakeLinearColorFromVectorAndFloat(
     "Vector (RGB)"=&MakeVectorFromLinearColorRGB(Color=&MergeHelper_ColorSwitch(
       0=<rgba 2 0 0.3495 1>
       1=<rgba 1 0.5739 0 1>
       2=?
       3=?
       Index=&MakeIntFromFloat(Float=$User.VariantPicker)
       Num=::NewEnumerator0
      ))
     "Float (Alpha)"=&Multiply_Float(A=0.5, B=&ClampFloat(Float=0.272))
    )
    Particles.GlowSpriteSize:Vector2f=&Vector2DFromFloat(Value=40)
    Particles.ShadowColor:LinearColor=&MakeLinearColorFromVectorAndFloat(
     "Vector (RGB)"=&MakeVectorFromLinearColorRGB(Color=&MergeHelper_ColorSwitch(
       0=<rgba 0 0.0118 0.3958 1>
       1=<rgba 0.3958 0.0826 0 1>
       2=?
       3=?
       Index=&MakeIntFromFloat(Float=$User.VariantPicker, Conversion Mode=::NewEnumerator1)
       Num=::NewEnumerator0
      ))
     "Float (Alpha)"=&Multiply_Float(
      A=1
      B=&Multiply_Float(
       A=&MergeHelper_FloatSwitch(0=1, 1=1, 2=?, 3=?, Index=&MakeIntFromFloat(Float=$User.VariantPicker), Num=::NewEnumerator0)
       B=&FNTimeOfDayScalar(Night Scale=1, Morning Scale=1, Evening Scale=0.8)
      )
     )
    )
    Particles.ShadowSize:Vector2f=&Vector2DFromFloat(Value=40)
   ) {isSetParameters=true}
   &SpriteFacingAndAlignment(
    Alignment Coordinate Space=::Local
    Facing Coordinate Space=::Local
    Sprite Alignment=<0 0 1>
    Sprite Facing=<0 -1 0>
    WriteAlignment=true
    WriteFacing=true
   )
   /SetVariables(
    Particles.HaloColor:LinearColor=&MakeLinearColorFromVectorAndFloat(
     "Vector (RGB)"=&Lerp_Vector(
      A=&MakeVectorFromLinearColorRGB(Color=<rgba 0.7527 0.1875 1 1>)
      B=&MakeVectorFromLinearColorRGB(Color=<rgba 1 0.4247 0.0938 1>)
      Alpha=$User.VariantPicker
     )
     "Float (Alpha)"=0.3
    )
    Particles.HaloDynamicMaterialParameter:Vector4f=&MakeVector4(X=1, Y=0.5, Z=0, W=0.8)
   ) {isSetParameters=true}
   &ScaleColor(
    Scale RGB=&MergeHelper_VectorSwitch(
     0=&Lerp_Vector(A=<0.6 0.6 0.6>, B=<0.9 0.9 0.9>, Alpha=$System.Timeline.Time)
     1=&Lerp_Vector(A=<0.6 0.6 0.6>, B=<0.8 0.8 0.8>, Alpha=$System.Timeline.Time)
     2=?
     3=?
     Index=&MakeIntFromFloat(Float=$User.VariantPicker)
     Num=::NewEnumerator0
    )
    ScaleA=true
    ScaleRGB=true
   )
   /SetVariables(Particles.EdgeColor:LinearColor=&MakeLinearColorFromVectorAndFloat(
     "Vector (RGB)"=&MakeVectorFromLinearColorRGB(Color=&MergeHelper_ColorSwitch(
       0=&LerpLinearColors(StartColor=<rgba 0.586 0.0626 4 1>, EndColor=<rgba 0.7325 0.0783 5 1>, LerpFactor=$System.Timeline.Time)
       1=&LerpLinearColors(StartColor=<rgba 1 0.3238 0.0157 1>, EndColor=<rgba 1.2 0.3886 0.0188 1>, LerpFactor=$System.Timeline.Time)
       2=?
       3=?
       Index=&MakeIntFromFloat(Float=$User.VariantPicker, Conversion Mode=::NewEnumerator1)
       Num=::NewEnumerator0
      ))
     "Float (Alpha)"=&Multiply_Float(
      A=&MergeHelper_FloatSwitch(0=1, 1=1, 2=?, 3=?, Index=&MakeIntFromFloat(Float=$User.VariantPicker), Num=::NewEnumerator0)
      B=&FNTimeOfDayScalar(Night Scale=2, Morning Scale=3, Evening Scale=3)
     )
    )) {isSetParameters=true}
   /SetVariables(Particles.OutineColor:LinearColor=&MakeLinearColorFromVectorAndFloat(
     "Vector (RGB)"=&MakeVectorFromLinearColorRGB(Color=&MergeHelper_ColorSwitch(
       0=/Niagara/DynamicInputs/Lerp/SmoothLerpOverTime_Color(
        Smooth Value:LinearColor=<rgba 0.9848 0.1028 5 1>
        Target Value:LinearColor=<rgba 1.5757 0.1645 8 1>
        Convergence Rate:NiagaraFloat=$System.Timeline.Time
       )
       1=&SmoothLerpOverTime_Color(
        Smooth Value:LinearColor=<rgba 3 1.9013 0.0617 1>
        Target Value:LinearColor=<rgba 5 2.9134 0.1028 1>
        Convergence Rate:NiagaraFloat=$System.Timeline.Time
       )
       2=?
       3=?
       Index=&MakeIntFromFloat(Float=$User.VariantPicker, Conversion Mode=::NewEnumerator1)
       Num=::NewEnumerator0
      ))
     "Float (Alpha)"=&Multiply_Float(
      A=&FNTimeOfDayScalar(Night Scale=0.4, Morning Scale=0.5, Day Scale=0.4, Evening Scale=0.4)
      B=&MergeHelper_FloatSwitch(0=0.5, 1=0.2, 2=?, 3=?, Index=&MakeIntFromFloat(Float=$User.VariantPicker), Num=::NewEnumerator0)
     )
    )) {isSetParameters=true}
   /SetVariables(Particles.OutlineScale:Vector3f=&Multiply_Vector(A=<1.12 1.12 1.12>, B=$Engine.Owner.Scale)) {isSetParameters=true}
   &FN_Light_Attributes(
    Color=&MergeHelper_ColorSwitch(
     0=<rgba 0.5702 0.2396 1 1>
     1=<rgba 1 0.4174 0 1>
     2=?
     3=?
     Index=&MakeIntFromFloat(Float=$User.VariantPicker)
     Num=::NewEnumerator0
    )
    Create Color Attribute=true
    Create Exponent Attribute=true
    Create Position Attribute=true
    Create Radius Attribute=true
    Exponent=5
    Position:NiagaraPosition=$Output.SkeletalMeshLocation.SampledPositionWithoutOffset
    Radius=50
    Scale Color=&Lerp_Float(A=10, B=20, Alpha=$System.Timeline.Time)
   )
  ]
  rendererClasses [
   &NiagaraSpriteRendererProperties
   &NiagaraLightRendererProperties
   &NiagaraMeshRendererProperties
  ]
  renderers [
   {rendererClass=&NiagaraSpriteRendererProperties}
   {rendererClass=&NiagaraSpriteRendererProperties, rendererIndex=1}
   {rendererClass=&NiagaraSpriteRendererProperties, rendererIndex=2}
   {rendererClass=&NiagaraSpriteRendererProperties, rendererIndex=3}
   {rendererClass=&NiagaraLightRendererProperties, rendererIndex=4}
   {rendererClass=&NiagaraSpriteRendererProperties, rendererIndex=5}
   {rendererClass=&NiagaraMeshRendererProperties, rendererIndex=6}
  ]
 }
 Galaxy {
  emitterUpdateScript [
   &EmitterState(Life Cycle Mode=::NewEnumerator1, Loop Duration=3.7167, UseLoopDelay=true)
   &SpawnRate(SpawnGroup=1, SpawnRate=5, Use Spawn Probability=false) {module=SpawnRate001}
   &SpawnRate(
    SpawnRate=&MakeFloatFromInt(Int=&MakeCustomIntFromBool(
      Bool=&SetBoolByFloatComparison(A=$System.Timeline.Time, B=0.1, Comparison Type=::NewEnumerator2)
      True INT=2
      False INT=2
     ))
    Use Spawn Probability=false
   )
  ]
  particleSpawnScript [
   &InitializeParticle(
    Color=&MergeHelper_ColorSwitch(
     0=<rgba 19.0969 12.6042 20 1>
     1=<rgba 15 12.911 5.3906 1>
     2=?
     3=?
     Index=&MakeIntFromFloat(Float=$User.VariantPicker, Conversion Mode=::NewEnumerator1)
     Num=::NewEnumerator0
    )
    Mesh Scale=&Multiply_VectorByFloat(Vector=<0.72 0.1 0.72>, Float=0.85)
    Mesh Scale Mode=::NewEnumerator4
    Sprite Size Mode=::NewEnumerator3
    Uniform Sprite Size=55
    UsePositionOffset=true
    Write Material Random=true
   )
   &InitialMeshOrientation(
    Mesh Orientation Mode=::NewEnumerator2
    Rotation=&RandomRangeVector(Minimum=<0 0 0>, Maximum=<0 360 0>)
    Use Rotation Vector=true
   )
   /CascadeToNiagaraConverter/NiagaraScripts/ModuleScripts/AddVectorToPosition(Vector:Vector3f=&TransformVector(Vector=<0 -15 -5>, Destination Space=::Simulation)) {module=AddVectorToPosition001}
   /SetVariables(Particles.VisibilityTag=&SpawnGroup()) {isSetParameters=true}
   /SetVariables() {isSetParameters=true}
  ]
  particleUpdateScript [
   /SetVariables(Particles.MeshRotationRate:Vector3f=$Particles.MeshRotationRateBase) {isSetParameters=true}
   &ParticleState()
   &ScaleSpriteSize(
    Scale Sprite Size Mode=::NewEnumerator0
    Uniform Scale Factor=&MakeFloatFromVector(Vector=$Engine.Owner.Scale, Channel=::NewEnumerator0)
   )
   &ScaleMeshSize(Scale Factor=$Engine.Owner.Scale)
   &ScaleColor(
    Scale Alpha=&FloatFromCurve(FloatCurve=DataInterface() {propertyValues {Curve {
        defaultValue=0
        keys [
         {interpMode=RCIM_Cubic, tangentMode=RCTM_User, time=0, value=0}
         {interpMode=RCIM_Cubic, tangentMode=RCTM_User, time=0.5, value=1}
         {tangentMode=RCTM_User, time=1, value=0}
        ]
       }}})
    ScaleA=true
    ScaleRGB=true
   )
   &DynamicMaterialParameters(
    Index 0 Param 1=0.5
    Index 0 Param 2=2
    Index 0 Param 4=1.3
    Module.Index0Float0Write=true
    Module.Index0Float1Write=true
    Module.Index0Float2Write=true
    Module.Index0Float3Write=true
   )
   /CascadeToNiagaraConverter/NiagaraScripts/ModuleScripts/UpdateMeshOrientationByEuler(RotationRate:Vector3f=$Particles.MeshRotationRate)
   &SkeletalMeshLocation(
    MaskSpawnGroup=false
    Mesh Sampling Type=::NewEnumerator5
    Sampled Position Offset=<0 -0.286 0>
    Skeletal Mesh=$System.Skeletal Mesh Reader
    Socket Sampling Mode=::NewEnumerator5
   )
   &MeshRotationForce(Rotation=&Lerp_Vector(A=<0 -10 0>, B=<0 -90 0>, Alpha=$System.Timeline.Time))
   &SolveForcesAndVelocity(Clamp Velocity=false, Limit Acceleration=false)
   &SpriteFacingAndAlignment(Sprite Alignment=<0 0 1>, Sprite Facing=<0 -1 0>, WriteAlignment=true, WriteFacing=true)
   /SetVariables(Particles.ReactivityDynamicMaterialParameter:Vector4f=&MakeVector4(X=&Lerp_Float(A=0.5, B=0.85, Alpha=$System.Timeline001.Time), Y=0, Z=0, W=0)) {isSetParameters=true}
  ]
  rendererClasses [&NiagaraMeshRendererProperties, &NiagaraSpriteRendererProperties]
  renderers [
   {rendererClass=&NiagaraMeshRendererProperties, rendererIndex=0}
   {rendererClass=&NiagaraSpriteRendererProperties, rendererIndex=1}
  ]
 }
 RotatingParticles {
  emitterUpdateScript [
   &EmitterState(
    Inactive Response=::NewEnumerator1
    Life Cycle Mode=::NewEnumerator1
    Loop Behavior=::NewEnumerator1
    Loop Duration Mode=::NewEnumerator1
    Reset Age On Awaken=true
    Scalability Mode=::NewEnumerator1
    UseLoopDelay=false
   )
   &SpawnBurst_Instantaneous(Spawn Count=8, Use Spawn Probability=false)
  ]
  particleSpawnScript [
   &InitializeParticle(
    Color=&MergeHelper_ColorSwitch(
     0=&LerpLinearColors(StartColor=<rgba 4.7537 2.5781 5 1>, EndColor=<rgba 18.9726 9.8958 20 1>, LerpFactor=$System.Timeline.Time)
     1=&LerpLinearColors(StartColor=<rgba 5 2.2092 0.9375 1>, EndColor=<rgba 5 2.2092 0.9375 1>, LerpFactor=$System.Timeline.Time)
     2=?
     3=?
     Index=&MakeIntFromFloat(Float=$User.VariantPicker, Conversion Mode=::NewEnumerator1)
     Num=::NewEnumerator0
    )
    Sprite Size Mode=::NewEnumerator3
    Uniform Sprite Size=2
    UsePositionOffset=true
    Write Material Random=true
   ) {module=InitializeParticle002}
   &ApplyInitialForces()
   &DynamicMaterialParameters(
    Index 0 Param 1=0.1
    Index 0 Param 2=&UniformRangedFloat(Minimum=0.025, Maximum=0.15)
    Index 0 Param 3=1
    Module.Param0WriteEnabled=true
    Module.Param1WriteEnabled=true
    Module.Param2WriteEnabled=true
    Module.Param3WriteEnabled=true
   )
   /SetVariables(Particles.RotateAroundPointPhase:NiagaraFloat=&UniformRangedFloat(Minimum=0, Maximum=360)) {isSetParameters=true}
   /SetVariables(Particles.RotateAroundPointZCenter:NiagaraFloat=&UniformRangedFloat(Minimum=-3, Maximum=3)) {isSetParameters=true}
  ]
  particleUpdateScript [
   &ParticleState(Kill Particles When Lifetime Has Elapsed=false, Let Infinitely Lived Particles Die When Emitter Deactivates=true)
   &RotateAroundPoint(
    Radius:NiagaraFloat=&RandomRangeFloat(Minimum=10, Maximum=16)
    Rotation Center:NiagaraPosition=&MakeVector(X=0, Y=-6, Z=$Particles.RotateAroundPointZCenter)
    Rotation Phase:NiagaraFloat=$Particles.RotateAroundPointPhase
    Rotation Rate:NiagaraFloat=&Lerp_Float(A=-0.1, B=-0.3, Alpha=$System.Timeline.Time)
    X Axis:Vector3f=<0 0 1>
    Y Axis:Vector3f=<1 0 0>
   )
   &SolveForcesAndVelocity(Clamp Velocity=false, Limit Acceleration=false)
   &ScaleColor(
    Scale RGB=&VectorFromFloat(Value=&Lerp_Float(A=1, B=2, Alpha=$System.Timeline.Time))
    ScaleA=true
    ScaleRGB=true
   )
  ]
  rendererClasses [&NiagaraSpriteRendererProperties]
  renderers [{rendererClass=&NiagaraSpriteRendererProperties}]
 }
 ShinnyDustAndWave {
  emitterUpdateScript [
   &EmitterState()
   &SpawnRate(
    SpawnGroup=1
    SpawnRate=&MakeFloatFromInt(Int=&MakeCustomIntFromBool(
      Bool=&SetBoolByFloatComparison(A=$System.Timeline.Time, B=0.1, Comparison Type=::NewEnumerator2)
      True INT=5
      False INT=0
     ))
    Use Spawn Probability=false
   ) {module=SpawnRate001}
   &SpawnRate(SpawnRate=15, Use Spawn Probability=false)
  ]
  particleSpawnScript [
   &InitializeParticle(
    Lifetime=2
    Sprite Rotation Mode=::NewEnumerator2
    Sprite Size Mode=::NewEnumerator3
    Sprite UV Mode=::NewEnumerator3
    Uniform Sprite Size=&MergeHelper_FloatSwitch(
     0=&Multiply_Float(A=&RandomRangeFloat(Minimum=40, Maximum=50), B=0.5)
     1=&Multiply_Float(A=&RandomRangeFloat(Minimum=50, Maximum=70), B=0.9)
     2=?
     3=?
     Index=&SpawnGroup()
     Num=::NewEnumerator0
    )
    UsePositionOffset=false
    Write Material Random=true
   )
   &CameraOffset(Camera Offset Amount=&RandomRangeFloat(Minimum=0, Maximum=30))
   /SetVariables(Particles.VisibilityTag=&SpawnGroup()) {isSetParameters=true}
  ]
  particleUpdateScript [
   &ParticleState()
   &SkeletalMeshLocation(
    MaskSpawnGroup=false
    Mesh Sampling Type=::NewEnumerator5
    OverrideRandomSeed=false
    Sampled Position Offset=&MergeHelper_VectorSwitch(
     0=/Niagara/DynamicInputs/Vectors/Position/AddVectorToPosition(Position=&RandomRangeVector(Minimum=<-8 0 -8>, Maximum=<8 0 8>), Vector=$Engine.Owner.Scale)
     1=<0 0.2722 0>
     2=?
     3=?
     Index=$Particles.VisibilityTag
     Num=::NewEnumerator0
    )
    Skeletal Mesh=$System.Skeletal Mesh Reader
   )
   /SetVariables(
    Particles.DustColor:LinearColor=&MakeLinearColorFromVectorAndFloat(
     "Vector (RGB)"=&MakeVectorFromLinearColorRGB(Color=&MergeHelper_ColorSwitch(
       0=<rgba 0.8858 0.4662 1 1>
       1=<rgba 1 0.8561 0.4662 1>
       2=?
       3=?
       Index=&MakeIntFromFloat(Float=$User.VariantPicker)
       Num=::NewEnumerator0
      ))
     "Float (Alpha)"=&FloatFromCurve(
      FloatCurve=DataInterface() {propertyValues {Curve {keys [<at 0 0 RCIM_Cubic>, {interpMode=RCIM_Cubic, tangentMode=RCTM_Break, time=0.2, value=1}, <at 1 0>]}}}
      Scale Curve=&Multiply_Float(
       A=&Multiply_Float(
        A=&FN_FrontEndFloatModifier(Gameplay=0.6, Front End=0.3)
        B=&self:SpawnGrp_0_1_Mask(
         Masked Float:NiagaraFloat=&RandomRangeFloat(Minimum=0.2, Maximum=0.4)
         Passthrough Float:NiagaraFloat=1
        )
       )
       B=&Multiply_Float(
        A=&FNTimeOfDayScalar(Night Scale=0.8, Morning Scale=1, Evening Scale=1)
        B=&FN_PickFloatDependOnPlatform(TargetPlatformActive=1, TargetPlatformNoActive=0.8)
       )
      )
     )
    )
    Particles.DustSize:Vector2f=&Multiply_Vector2D(
     A=$Particles.SpriteSize
     B=&Vector2DFromFloat(Value=&Multiply_Float(
       A=&FloatFromCurve(FloatCurve=DataInterface() {propertyValues {Curve {keys [
            {arriveTangent=3.5003, interpMode=RCIM_Cubic, leaveTangent=3.5003, tangentMode=RCTM_User, time=0, value=0.5}
            {arriveTangent=0.5833, interpMode=RCIM_Cubic, leaveTangent=0.5833, time=0.18, value=0.9}
            <at 1.2 1.2 RCIM_Cubic>
           ]}}})
       B=&MakeFloatFromVector(Vector=$Engine.Owner.Scale, Channel=::NewEnumerator0)
      ))
    )
   ) {isSetParameters=true}
   /SetVariables(
    Particles.WaveColor:LinearColor=&MakeLinearColorFromVectorAndFloat(
     "Vector (RGB)"=&MakeVectorFromLinearColorRGB(Color=&MergeHelper_ColorSwitch(
       0=<rgba 0.2927 0.0457 0.5104 1>
       1=<rgba 0.3802 0.1852 0.0455 1>
       2=?
       3=?
       Index=&MakeIntFromFloat(Float=$User.VariantPicker)
       Num=::NewEnumerator0
      ))
     "Float (Alpha)"=&FloatFromCurve(
      FloatCurve=DataInterface() {propertyValues {Curve {keys [
          {arriveTangent=0.0516, interpMode=RCIM_Cubic, time=-0.003, value=0}
          {arriveTangent=-0.7399, interpMode=RCIM_Cubic, leaveTangent=-0.7399, tangentMode=RCTM_User, time=0.2003, value=0.2468}
          <at 1 0>
         ]}}}
      Scale Curve=0.5
     )
    )
    Particles.WaveDynamicMaterialParameter:Vector4f=&MakeVector4(X=0, Y=0, Z=0.1, W=0)
    Particles.WaveSize:Vector2f=&Multiply_Vector2D(
     A=$Particles.SpriteSize
     B=&Vector2DFromFloat(Value=&Multiply_Float(
       A=&FloatFromCurve(FloatCurve=DataInterface() {propertyValues {Curve {
           defaultValue=0
           keys [
            {arriveTangent=7.359, interpMode=RCIM_Cubic, leaveTangent=7.359, tangentMode=RCTM_User, time=0, value=0}
            {interpMode=RCIM_Cubic, tangentMode=RCTM_User, time=0.2664, value=1}
           ]
          }}})
       B=&MakeFloatFromVector(Vector=$Engine.Owner.Scale, Channel=::NewEnumerator0)
      ))
    )
   ) {isSetParameters=true}
   &SpriteFacingAndAlignment(Sprite Alignment=<0 0 1>, Sprite Facing=<0 -1 0>, WriteAlignment=true, WriteFacing=true)
  ]
  rendererClasses [&NiagaraSpriteRendererProperties]
  renderers [
   {rendererClass=&NiagaraSpriteRendererProperties}
   {rendererClass=&NiagaraSpriteRendererProperties, rendererIndex=1}
  ]
 }
 Sparkles {
  emitterUpdateScript [
   &EmitterState()
   &SpawnRate(
    SpawnRate=&MakeFloatFromInt(Int=&MakeCustomIntFromBool(
      Bool=&SetBoolByFloatComparison(A=$System.Timeline.Time, B=0.1, Comparison Type=::NewEnumerator2)
      True INT=5
      False INT=2
     ))
    Use Spawn Probability=false
   )
  ]
  particleSpawnScript [&InitializeParticle(
    Color=&MergeHelper_ColorSwitch(
     0=<rgba 1.8774 0.4705 2 1>
     1=<rgba 2 1.8271 0.4705 1>
     2=?
     3=?
     Index=&MakeIntFromFloat(Float=$User.VariantPicker, Conversion Mode=::NewEnumerator1)
     Num=::NewEnumerator0
    )
    Lifetime=&RandomRangeFloat(Minimum=0.5, Maximum=0.7)
    Position=<0 -6 0>
    Position Mode=::NewEnumerator1
    Sprite Size=&Multiply_Vector2D_ByFloat(Float=&RandomRangeFloat(Minimum=0.5, Maximum=1), Vector2D=<15 15>)
    Sprite Size Mode=::NewEnumerator4
    UsePositionOffset=false
    Write Material Random=true
   )]
  particleUpdateScript [
   &ParticleState()
   &ShapeLocation(
    MaskSpawnGroup=false
    Offset=<0 -6 0>
    Offset Mode=::NewEnumerator0
    Override Seed=false
    Ring Radius=14.5
    Rotation Axis=<0 1 0>
    Rotation Mode=::NewEnumerator1
    Scale Mode:ENiagara_ScaleMode=::NewEnumerator1
    Shape Primitive=::NewEnumerator4
   )
   &ScaleColor(
    Scale Alpha=&FloatFromCurve(
     FloatCurve=DataInterface() {propertyValues {Curve {keys [
         {interpMode=RCIM_Cubic, tangentMode=RCTM_User, time=0, value=0}
         {arriveTangent=-2.4058, interpMode=RCIM_Cubic, leaveTangent=-2.4058, tangentMode=RCTM_User, time=0.26, value=0.8421}
         {arriveTangent=-1.1775, interpMode=RCIM_Cubic, leaveTangent=-1.1775, time=0.4613, value=0.0645}
         {arriveTangent=-0.1197, interpMode=RCIM_Cubic, leaveTangent=-0.1197, time=0.5898, value=0.4537}
         <at 1 0 RCIM_Cubic>
        ]}}}
     Scale Curve=&RandomRangeFloat(Minimum=0.5, Maximum=1)
    )
    Scale RGB=&MakeVectorFromLinearColorRGB(Color=&MergeHelper_ColorSwitch(
      0=&LerpLinearColors(StartColor=<rgba 30.5277 6.6127 50 1>, EndColor=<rgba 42.7387 9.2577 70 1>, LerpFactor=$System.Timeline.Time)
      1=&LerpLinearColors(StartColor=<rgba 50 33.7768 6.6127 1>, EndColor=<rgba 70 47.2876 9.2577 1>, LerpFactor=$System.Timeline.Time)
      2=?
      3=?
      Index=&MakeIntFromFloat(Float=$User.VariantPicker, Conversion Mode=::NewEnumerator1)
      Num=::NewEnumerator0
     ))
    ScaleA=true
    ScaleRGB=true
   )
   &ScaleSpriteSize(Scale Factor=&Vector2DFromFloat(Value=&FloatFromCurve(FloatCurve=DataInterface() {propertyValues {Curve {
         defaultValue=1
         keys [
          {interpMode=RCIM_Cubic, tangentMode=RCTM_User, time=0, value=0}
          {interpMode=RCIM_Cubic, tangentMode=RCTM_User, time=0.5, value=1}
          {tangentMode=RCTM_User, time=1, value=0}
         ]
        }}})))
   &ScaleColor(
    Scale Alpha=&FNTimeOfDayScalar(Night Scale=0.8, Morning Scale=1, Evening Scale=1)
    ScaleA=true
    ScaleRGB=true
   ) {module=ScaleColor002}
  ]
  rendererClasses [&NiagaraSpriteRendererProperties]
  renderers [{rendererClass=&NiagaraSpriteRendererProperties}]
 }
}
summary {
 systemName=NS_Backpack_IndoorLaceBeat
 emitters [
  {emitterName=Sparkles, rendererClasses [&NiagaraSpriteRendererProperties]}
  {
   emitterName=Clock
   rendererClasses [
    &NiagaraSpriteRendererProperties
    &NiagaraLightRendererProperties
    &NiagaraMeshRendererProperties
   ]
  }
  {emitterName=RotatingParticles, rendererClasses [&NiagaraSpriteRendererProperties]}
  {emitterName=Galaxy, rendererClasses [&NiagaraMeshRendererProperties, &NiagaraSpriteRendererProperties]}
  {emitterName=ShinnyDustAndWave, rendererClasses [&NiagaraSpriteRendererProperties]}
 ]
 userVariables [
  User.bIsFrontEnd {type=NiagaraBool}
  User.bIsFrontEndPreview {type=NiagaraBool}
  User.DebugReactivity {type=NiagaraBool}
  User.EnableFrontendFakeAudioLoop {type=NiagaraFloat, defaultValue=0, description=""}
  User.VariantPicker {type=NiagaraFloat, defaultValue=0, description=""}
 ]
}
